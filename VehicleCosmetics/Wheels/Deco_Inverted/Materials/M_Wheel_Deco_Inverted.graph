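# Material for the Deco Inverted vehicle wheel cosmetic, giving it glowing tread and rim detail over a flat-normal, fully rough, non-metallic surface.
# Vertex colors and the ID, Tread and Rim textures supply masks for emissive layers, edge lightning and Fresnel glow, with time-panned noise animating two layers.
# A Painted static switch swaps the purple and green palette for Color 3.
# Named reroutes link the stages, and the Fresnel function comes from the Cola_Holographic wheel.
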Material /VehicleCosmetics/Wheels/Deco_Inverted/Materials/M_Wheel_Deco_Inverted
graph {
 properties {
  bAutomaticallySetUsageInEditor=false
  bUseMaterialAttributes=true
  bUsedWithMeshDeformer=true
  bUsedWithSkeletalMesh=true
  previewMesh=/VehicleCosmetics/Wheels/Deco_Inverted/New/Meshes/DecoInvertedLOD2
  usedShadingModels=Default Lit
 }
 nodes [
  n1 = VertexColor()
  NamedRerouteDeclaration(None=@n1.R) {name=VC_R, nodeColor=<rgba 0.1719 0 1 1>}
  NamedRerouteDeclaration(None=@n1.G) {name=VC_G, nodeColor=<rgba 0 1 0 1>}
  n2 = Multiply(A=Multiply(A=NamedRerouteUsage(), B=NamedRerouteUsage()), B=NamedRerouteUsage())
  n3 = ScalarParameter() {defaultValue=15, parameterName=Emissive Tread B Intensity}
  n4 = VectorParameter() {parameterName=Mask 3 Transforms, defaultValue=<rgba 2 1 -0.5 0>}
  n5 = Add(A=Multiply(A=TextureCoordinate(), B=AppendVector(A=@n4.R, B=@n4.R)), B=Multiply(A=AppendVector(A=@n4.B, B=@n4.B), B=Time()))
  NamedRerouteDeclaration(None=Multiply(
    A=Multiply(A=@n2, B=@n3)
    B=TextureSample(UVs=@n5) {samplerSource=SSM_Wrap_WorldGroupSettings, texture=/VehicleCosmetics/Wheels/_SharedTextures/FX/Universal_Noise_66_RGB}.R
   )) {name=Mask 3 Out, nodeColor=<rgba 0.6172 1 0 1>}
  n6 = TextureSampleParameter2D() {parameterName=ID, samplerType=SAMPLERTYPE_LinearColor, texture=/VehicleCosmetics/Wheels/Deco_Inverted/Textures/DecoInvertedID}
  NamedRerouteDeclaration(None=@n6.R) {name=Wheel_Masks_R, nodeColor=<rgba 0 0.7891 1 1>}
  n7 = TextureSampleParameter2D() {parameterName=Tread, samplerType=SAMPLERTYPE_LinearColor, texture=/VehicleCosmetics/Wheels/Deco_Inverted/Textures/T_DecoInverted_Tire_M}
  n8 = NamedRerouteUsage()
  n9 = Multiply(A=Multiply(A=@n7.G, B=@n8)) {constB=2}
  n10 = TextureSampleParameter2D() {parameterName=Rim, samplerType=SAMPLERTYPE_LinearColor, texture=/VehicleCosmetics/Wheels/Deco_Inverted/Textures/T_DecoInverted_Rim_M_LOD}
  n11 = NamedRerouteUsage()
  n12 = Add(A=NamedRerouteUsage(), B=@n11)
  n13 = NamedRerouteUsage()
  n14 = Multiply(A=@n10.R, B=OneMinus(None=Add(A=@n12, B=@n13)))
  n15 = Add(A=Clamp(None=LinearInterpolate(A=@n10.R, B=@n14, Alpha=@n11)), B=Clamp(None=Constant()))
  n16 = NamedRerouteUsage()
  n17 = Multiply(A=@n16, B=NamedRerouteUsage())
  n18 = Add(
   A=LinearInterpolate(A=@n9, B=Constant(), Alpha=@n8)
   B=LinearInterpolate(B=LinearInterpolate(A=@n15, Alpha=@n16) {constB=0}, Alpha=OneMinus(None=@n17)) {constB=0}
  )
  NamedRerouteDeclaration(None=Multiply(A=Multiply(A=@n18) {constB=5}, B=NamedRerouteUsage()) {constB=5}) {name=Edge Lightning, nodeColor=<rgba 0 1 0.5781 1>}
  n19 = NamedRerouteUsage()
  n20 = NamedRerouteUsage()
  n21 = Multiply(A=NamedRerouteUsage(), B=Max(A=Multiply(A=@n20) {constB=0.7}) {constB=0.05})
  n22 = Max(A=Max(A=@n21, B=NamedRerouteUsage()), B=NamedRerouteUsage())
  n23 = Clamp(None=Add(A=Max(A=@n19, B=@n22), B=NamedRerouteUsage())) {maxDefault=3}
  n24 = Clamp(None=NamedRerouteUsage()) {maxDefault=3}
  NamedRerouteDeclaration(None=LinearInterpolate(A=Add(A=@n23, B=@n24), Alpha=NamedRerouteUsage()) {constB=0}) {name=Emissive, nodeColor=<rgba 1 0.0938 0 1>}
  n25 = Multiply(
   A=MaterialFunctionCall!(Normal, Exponent=Constant() {r=1.5}) {materialFunction=/VehicleCosmetics/Wheels/Cola_Holographic/Materials/MF_Fresnel_Classic}.Result
   B=ComponentMask(None=@n21) {b=true, g=true, r=true}
  )
  NamedRerouteDeclaration(None=LinearInterpolate(A=Multiply(A=@n25) {constB=5}, Alpha=NamedRerouteUsage()) {constB=0.05}) {name=Color, nodeColor=<rgba 1 0.0938 0 1>}
  n26 = NamedRerouteUsage()
  n27 = NamedRerouteUsage()
  n28 = OneMinus(None=Add(A=Add(A=@n26, B=@n27), B=NamedRerouteUsage()))
  n29 = Multiply(A=@n28) {constB=2}
  n30 = MaterialFunctionCall(Exponent=Constant() {r=2.5}) {materialFunction=&MF_Fresnel_Classic}
  n31 = Multiply(A=Multiply(A=@n28, B=ConstantBiasScale(None=@n30.Result) {bias=-0.5, scale=4})) {constB=2}
  n32 = LinearInterpolate(A=Clamp(None=Add(A=@n29, B=@n31)), Alpha=NamedRerouteUsage()) {constB=0}
  n33 = ScalarParameter() {defaultValue=1, parameterName=Fresnel Intensity}
  NamedRerouteDeclaration(None=Multiply(A=Multiply(A=@n32, B=@n33), B=NamedRerouteUsage())) {name=Fresnel, nodeColor=<rgba 1 0 0.7969 1>}
  NamedRerouteDeclaration(None=@n10.B) {name=Blue, nodeColor=<rgba 0 1 0.4141 1>}
  NamedRerouteDeclaration(None=@n10.A) {name=AlhaLOD, nodeColor=<rgba 1 0.1641 0 1>}
  NamedRerouteDeclaration(None=@n1.B) {name=VC_B, nodeColor=<rgba 0 0 1 1>}
  n34 = NamedRerouteUsage()
  n35 = Multiply(A=NamedRerouteUsage(), B=NamedRerouteUsage())
  n36 = Constant()
  n37 = NamedRerouteUsage()
  n38 = LinearInterpolate(
   A=Multiply(A=Multiply(A=@n34, B=@n35), B=ScalarParameter() {defaultValue=15, parameterName=Emissive Wheel R Intensity})
   B=@n36
   Alpha=LinearInterpolate(A=@n36, B=NamedRerouteUsage(), Alpha=OneMinus(None=@n37))
  )
  NamedRerouteDeclaration(None=OneMinus(None=NamedRerouteDeclaration(None=@n38) {name=Cut01, nodeColor=<rgba 1 0.2812 0 1>})) {name=Cut02, nodeColor=<rgba 1 0.2812 0 1>}
  NamedRerouteDeclaration(None=@n6.B) {name=Wheel_Masks_B, nodeColor=<rgba 0 0 1 1>}
  NamedRerouteDeclaration(None=@n6.A) {name=Wheel_Masks_A, nodeColor=<rgba 1 1 1 1>}
  NamedRerouteDeclaration(None=@n6.G) {name=Wheel_Masks_G, nodeColor=<rgba 0 1 0 1>}
  n39 = VectorParameter() {parameterName=Color 3, defaultValue=<rgba 0 0.0666 1 1>}
  n40 = VectorParameter() {parameterName=Purple, defaultValue=<rgba 0.0395 0.3005 0.2051 1>}
  NamedRerouteDeclaration(None=StaticSwitchParameter(True=@n39.RGB, False=@n40.RGB) {parameterName=Painted}) {name=Purple*Color3, nodeColor=<rgba 1 0.0234 0 1>}
  n41 = VectorParameter() {parameterName=Green, defaultValue=<rgba 0.4397 0.2705 0.0704 1>}
  n42 = LinearInterpolate(A=@n41.RGB, B=@n40.RGB, Alpha=Multiply(A=NamedRerouteUsage(), B=NamedRerouteUsage()))
  n43 = ScalarParameter() {defaultValue=1, parameterName=AlphaLodColor_M}
  n44 = NamedRerouteUsage()
  NamedRerouteDeclaration(None=LinearInterpolate(A=StaticSwitchParameter(True=@n39.RGB, False=@n42) {parameterName=Painted}, B=Multiply(A=@n41.RGB, B=@n43), Alpha=OneMinus(None=@n44))) {name=Final Color, nodeColor=<rgba 1 0.0234 0 1>}
  n45 = Constant()
  n46 = LinearInterpolate(A=OneMinus(None=@n9), B=Constant(), Alpha=NamedRerouteUsage())
  n47 = VectorParameter() {parameterName=Mask 4 Transforms, defaultValue=<rgba 4 1 -0.4 0>}
  n48 = Add(A=Multiply(A=TextureCoordinate(), B=AppendVector(A=@n47.R, B=@n47.R)), B=Multiply(A=AppendVector(A=@n47.B, B=@n47.B), B=Time()))
  NamedRerouteDeclaration(None=Multiply(
    A=LinearInterpolate(A=@n38, B=@n45, Alpha=@n46)
    B=TextureSample(UVs=@n48) {samplerSource=SSM_Wrap_WorldGroupSettings, texture=&Universal_Noise_66_RGB}.R
   )) {name=Mask 4 Out, nodeColor=<rgba 0.6172 1 0 1>}
  n49 = NamedRerouteUsage()
  n50 = NamedRerouteUsage()
  n51 = NamedRerouteUsage()
  n52 = ScalarParameter() {defaultValue=3, parameterName=Emissive Tread G Intensity}
  NamedRerouteDeclaration(None=Multiply(A=Multiply(A=@n49, B=@n50), B=Multiply(A=@n51, B=@n52))) {name=Mask 2 Out, nodeColor=<rgba 0.6172 1 0 1>}
  SetMaterialAttributes!(
   MaterialAttributes
   Base Color=NamedRerouteUsage()
   Emissive Color=NamedRerouteUsage()
   Normal=Constant3Vector() {constant=<rgba 0 0 1 0>}
   Metallic=Constant()
   Roughness=Constant() {r=1}
  ) {attributeSetTypes [
    #1
    #2
    #3
    #4
    #5
   ]}
 ]
}
