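# Material function in the Reflections library that samples a sphere-style reflection texture using a transformed reflection vector for alignment.
# The transformed vector's XY is remapped by half scale and bias into UVs, and the sampled RGB becomes WorldReflection.
# WorldReflectionShadowed multiplies that colour by a squared mask, built from the vector's Z through ConstantBiasScale and CheapContrast.
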MaterialFunction /Engine/Functions/Engine_MaterialFunctions01/Reflections/WorldAlignedReflection
graph {
 properties {
  bExposeToLibrary=true
  description="Alignes a Reflection texture to your view.  Reflection should be a sphere."
  libraryCategoriesText [Misc,Reflections]
 }
 nodes [
  n1 = Transform(None=FunctionInput() {bUsePreviewValueAsDefault=true, inputName=ReflectionVector})
  n2 = ComponentMask(None=@n1) {g=true, r=true}
  n3 = Constant() {r=0.5}
  n4 = TextureSample(
   UVs=Add(A=Multiply(A=@n2, B=@n3), B=@n3)
   Tex=FunctionInput(Preview=TextureObject() {texture=/Engine/EngineMaterials/DefaultDiffuse}) {description=Reflection should look like a sphere, inputName=ReflectionTexture, inputType=FunctionInput_Texture2D}
  ) {texture=&DefaultDiffuse}
  FunctionOutput(None=@n4.RGB) {outputName=WorldReflection}
  n5 = MaterialFunctionCall(In=ConstantBiasScale(None=ComponentMask(None=@n1) {b=true}), Contrast=Constant() {r=2}) {materialFunction=&CheapContrast}
  FunctionOutput(None=Multiply(A=@n4.RGB, B=Multiply(A=@n5.Result, B=@n5.Result))) {outputName=WorldReflectionShadowed, sortPriority=1}
 ]
}
