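# Tread material for the Deco Inverted cosmetic vehicle wheel, default lit and usable on skeletal meshes.
# A desaturated world-aligned cubemap reflection adds to a term built from the tread mask's green channel and a Fresnel function, forming an intensity-scaled color.
# A Painted static switch selects a single color or a Green-to-Purple blend, and vertex color channels are exposed through named reroutes.
# Material attributes set base color, metallic and roughness.
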Material /VehicleCosmetics/Wheels/Deco_Inverted/Materials/M_Wheel_Deco_Inverted_Tread
graph {
 properties {
  bAutomaticallySetUsageInEditor=false
  bUseMaterialAttributes=true
  bUsedWithMeshDeformer=true
  bUsedWithSkeletalMesh=true
  usedShadingModels=Default Lit
 }
 nodes [
  n1 = VertexColor()
  NamedRerouteDeclaration(None=@n1.R) {name=VC_R, nodeColor=<rgba 1 0.7266 0 1>}
  NamedRerouteDeclaration(None=@n1.G) {name=VC_G, nodeColor=<rgba 0 1 0 1>}
  n2 = TextureObject() {texture=/Game/Packages/Fortress_SharedMaterials/Textures/T_Cubemap_01_tex}
  n3 = Multiply(
   A=NamedRerouteUsage()
   B=Desaturation(
    None=MaterialFunctionCall!(ReflectionVector, ReflectionTexture=@n2) {materialFunction=/Engine/Functions/Engine_MaterialFunctions01/Reflections/WorldAlignedReflection}.WorldReflection
    Fraction=Constant() {r=1}
   ) {luminanceFactors=<rgba 0.3 0.59 0.11 0>}
  )
  n4 = TextureSampleParameter2D() {parameterName=Tread, texture=/VehicleCosmetics/Wheels/Deco_Inverted/Textures/Wheel_Deco_Inverted_Tread_Masks}
  n5 = ScalarParameter() {defaultValue=2, parameterName=Fresnel Exp}
  n6 = Clamp(None=Multiply(A=MaterialFunctionCall!(Normal, Exponent=@n5) {materialFunction=/VehicleCosmetics/Wheels/Cola_Holographic/Materials/MF_Fresnel_Classic}.Result) {constB=3})
  n7 = Multiply(A=Add(A=Multiply(A=@n4.G), B=Add(A=@n6) {constB=0.2}) {constB=0.2}, B=NamedRerouteUsage())
  NamedRerouteDeclaration(None=Multiply(A=Add(A=@n3, B=@n7), B=ScalarParameter() {defaultValue=1, parameterName=Emissive Wheel R Intensity})) {name=BC, nodeColor=<rgba 0 0.3437 1 1>}
  NamedRerouteDeclaration(None=@n1.B) {name=VC_B, nodeColor=<rgba 0 0 1 1>}
  n8 = VectorParameter() {parameterName=Green, defaultValue=<rgba 0.4397 0.2705 0.0704 1>}
  n9 = VectorParameter() {parameterName=Purple, defaultValue=<rgba 0.0395 0.3005 0.2051 1>}
  n10 = NamedRerouteUsage()
  NamedRerouteDeclaration(None=StaticSwitchParameter(
    True=VectorParameter() {parameterName=Color 3, defaultValue=<rgba 0 0.0666 1 1>}.RGB
    False=LinearInterpolate(A=@n8.RGB, B=@n9.RGB, Alpha=@n10)
   ) {parameterName=Painted}) {name=Final Color, nodeColor=<rgba 1 0.0234 0 1>}
  SetMaterialAttributes!(
   MaterialAttributes
   Base Color=NamedRerouteUsage()
   Metallic=ScalarParameter() {parameterName=Metallic}
   Roughness=ScalarParameter() {defaultValue=1, parameterName=Roughness}
  ) {attributeSetTypes [#1, #2, #3]}
 ]
}
